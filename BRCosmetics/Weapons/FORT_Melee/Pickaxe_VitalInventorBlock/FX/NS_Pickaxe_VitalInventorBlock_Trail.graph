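# Niagara ribbon trail effect for the VitalInventorBlock pickaxe, built as a single ColorTrail emitter.
# FN anim-trail modules spawn, initialize and update short-lived particles against the system's skeletal mesh, tinted at spawn by a time-of-day color.
# Over particle life, curves scale color and alpha and taper ribbon width late, while a set-parameters step overrides the ribbon V range.
# User parameters expose a gradient texture and ribbon ID offset.
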NiagaraSystem /BRCosmetics/Weapons/FORT_Melee/Pickaxe_VitalInventorBlock/FX/NS_Pickaxe_VitalInventorBlock_Trail
dependencies {
 usedDataInterfaces [&NiagaraDataInterfaceCurve, &NiagaraDataInterfaceSkeletalMesh]
 usedDynamicInputs []
 usedModules [
  &SystemState
  &self:SystemSpawnScript.SystemScriptSource.SystemScriptGraph.NiagaraNodeAssignment_0.SetVariables_#1
 ]
}
emitters {ColorTrail {
  emitterUpdateScript [
   &EmitterState(Life Cycle Mode=::NewEnumerator1, Loop Behavior=::NewEnumerator1, Loop Duration=0.3, UseLoopDelay=false)
   &FN_SpawnAnimTrail()
  ]
  particleSpawnScript [
   &FN_InitializeAnimTrailBoneIndecies()
   &FN_InitializeAnimTrail(
    Color=&FNTimeOfDayColor(
     Night Color=<rgba 0.1841 0.2186 0.6562 1>
     Morning Color=<rgba 0.8802 0.6223 0.3753 1>
     Day Color=<rgba 1 1 1 1>
     Evening Color=<rgba 0.526 0.1762 0.1026 1>
    )
    Lifetime=0.1
    SkeletalMesh=$System.Skeletal Mesh
   )
   &AddVelocity(Velocity=<0 -100 0>) {enabled=false}
  ]
  particleUpdateScript [
   &FN_UpdateAnimTrail(SkeletalMesh=$System.Skeletal Mesh) {module=FN_UpdateAnimTrail001}
   &ScaleColor(
    Scale Alpha=&FloatFromCurve(
     FloatCurve=DataInterface() {propertyValues {Curve {keys [
         {arriveTangent=-0.5448, interpMode=RCIM_Cubic, leaveTangent=-0.5448, tangentMode=RCTM_User, time=0, value=1}
         {arriveTangent=-0.643, interpMode=RCIM_Cubic, leaveTangent=-0.643, tangentMode=RCTM_User, time=1, value=0}
        ]}}}
     Scale Curve=0.22
    )
    Scale RGB=&Multiply_VectorByFloat(
     Vector=<1 1 1>
     Float=&FloatFromCurve(
      FloatCurve=DataInterface() {propertyValues {Curve {keys [{tangentMode=RCTM_User, time=0.5, value=1}, {tangentMode=RCTM_User, time=1, value=0.5}]}}}
      Scale Curve=0.5
     )
    )
    ScaleA=true
    ScaleRGB=true
   )
   /SetVariables(Particles.RibbonV0RangeOverride:Vector2f=<1 0>) {isSetParameters=true}
   &ScaleRibbonWidth(Ribbon Width Scale=&FloatFromCurve(FloatCurve=DataInterface() {propertyValues {Curve {keys [<at 0.75 1>,<at 1 0>]}}}))
  ]
  rendererClasses [&NiagaraRibbonRendererProperties]
  renderers [{rendererClass=&NiagaraRibbonRendererProperties}]
 }}
summary {
 systemName=NS_Pickaxe_VitalInventorBlock_Trail
 emitters [{emitterName=ColorTrail, rendererClasses [&NiagaraRibbonRendererProperties]}]
 userVariables [
  User.GradientTexture {
   type=Texture
   description=""
   defaultValue {
    struct=NiagaraExt_VariableValue_Object
    value {object=&/T_VitalInventor_Gradient, objectClass=/Script/Engine.Texture}
   }
  }
  User.RibbonIDOffset {type=NiagaraInt32}
 ]
}
