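# Control Rig for the VitalInventorBlock pickaxe that adjusts its two dynamic eye bones.
# Each execution runs a parent constraint per eye, blending the eye bone's own transform with its base bone, weighted toward the base.
# The constraints filter translation to Y and Z, leave rotation out and include scale.
# A final Modify Transforms step nudges the right eye along X; the ribbon bones are untouched.
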ControlRigBlueprint /BRCosmetics/Weapons/FORT_Melee/Pickaxe_VitalInventorBlock/Meshes/VitalInventorBlock_Axe_CtrlRig
graphs [
 RigVMFunctionLibrary {links [], nodes []}
 RigVMModel {nodes [
   RigUnit_BeginExecution = &RigUnit_BeginExecution()
   &RigUnit_ParentConstraint(
    AdvancedSettings=(InterpolationType=Shortest,RotationOrderForFilter=XYZ)
    Child=(Type=Bone,Name="dyn_eye_l")
    Filter=(TranslationFilter=(bX=false,bY=true,bZ=true),RotationFilter=(bX=false,bY=false,bZ=false),ScaleFilter=(bX=True,bY=True,bZ=True))
    Parents=((Item=(Type=Bone,Name="dyn_eye_l_base"),Weight=0.7),(Item=(Type=Bone,Name="dyn_eye_l"),Weight=0.3))
    Weight=1
    ExecutePin=@RigUnit_BeginExecution.ExecutePin
   )
   > &RigUnit_ParentConstraint(
    AdvancedSettings=(InterpolationType=Shortest,RotationOrderForFilter=ZYX)
    Child=(Type=Bone,Name="dyn_eye_r")
    Filter=(TranslationFilter=(bX=false,bY=true,bZ=true),RotationFilter=(bX=false,bY=false,bZ=false),ScaleFilter=(bX=True,bY=True,bZ=True))
    Parents=((Item=(Type=Bone,Name="dyn_eye_r_base"),Weight=0.7),(Item=(Type=Bone,Name="dyn_eye_r"),Weight=0.3))
    Weight=1
    bMaintainOffset=true
   )
   > &RigUnit_ModifyTransforms(
    ItemToModify=((Item=(Type=Bone,Name="dyn_eye_r"),Transform=(Rotation=(X=0,Y=0,Z=0,W=1),Translation=(X=-0.3,Y=0,Z=0),Scale3D=(X=1,Y=1,Z=1))))
    Weight=1
   )
  ]}
]
hierarchy [
 b_AR_Root {loc=<0 0 0>, rot=<0 0 0.7071 0.7071>}
 Melee_TwoHanded_Handle {parent=b_AR_Root, loc=<8.1477 -0.008 -29.0793>, rot=<0 0 0.7071 -0.7071>}
 dyn_eye_l_base {parent=Melee_TwoHanded_Handle, loc=<2 8.3523 85.5793>, rot=<0 0 0 1>}
 dyn_eye_r_base {parent=Melee_TwoHanded_Handle, loc=<0 8.3523 85.5793>, rot=<0 0 0.7071 -0.7071>}
 dyn_eye_l {parent=Melee_TwoHanded_Handle, loc=<4.9615 7.2721 86.5916>, rot=<0 -0.0273 0 -0.9996>}
 dyn_eye_r {parent=Melee_TwoHanded_Handle, loc=<-2.9472 7.2721 86.5907>, rot=<0 0.0281 0 -0.9996>}
 dyn_ribbon_01 {parent=Melee_TwoHanded_Handle, loc=<1.0147 -8.7835 71.2149>, rot=<0 0 0 1>}
 dyn_ribbon_02 {parent=Melee_TwoHanded_Handle, loc=<1.0147 -8.7835 69.3716>, rot=<0 0 0 1>}
]
