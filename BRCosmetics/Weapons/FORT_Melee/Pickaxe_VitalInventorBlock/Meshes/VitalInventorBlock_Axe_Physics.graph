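# Physics asset for the VitalInventorBlock two-handed pickaxe, giving collision shapes to the handle and its dynamic eye and ribbon bones.
# A long capsule covers the handle, small capsules cover the left and right eye bases, and thin rotated boxes cover two ribbon bones.
# Each dynamic bone is constrained directly to the handle rather than chained, with both eyes set to free swing and twist.
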PhysicsAsset /BRCosmetics/Weapons/FORT_Melee/Pickaxe_VitalInventorBlock/Meshes/VitalInventorBlock_Axe_Physics
bodies {
 Melee_TwoHanded_Handle [{shapeType=Capsule, length=80.9702, radius=8.0114, center=<1.0619 -0.0526 37.146>, rotation=<rot 1.5594 -89.9918 -0.0528>}]
 dyn_eye_l_base [{shapeType=Capsule, length=0.505, radius=2.5829, shapeName=dyn_eye_l_base_capsule, center=<1.4174 -0.7615 0>, rotation=<rot 0 -90 0>}]
 dyn_eye_r_base [{shapeType=Capsule, length=0.505, radius=2.5829, shapeName=dyn_eye_l_base_capsule, center=<0.3578 -1.5687 0.1579>, rotation=<rot 0 -90 0>}]
 dyn_ribbon_01 [{
   shapeType=Box
   extentX=5.5
   extentY=7.5
   extentZ=1.5
   shapeName=dyn_ribbon_01_box
   center=<0.3745 -2.0971 1.0561>
   rotation=<rot -18.8828 -1.0952 31.9106>
  }]
 dyn_ribbon_02 [{
   shapeType=Box
   extentX=4.5
   extentY=6.5
   extentZ=1.5
   shapeName=dyn_ribbon_02_box
   center=<0.9372 -1.6609 -0.3985>
   rotation=<rot 39.9189 10.9861 -10.1183>
  }]
}
constraints [
 {
  bone1Name=dyn_eye_l_base
  bone2Name=Melee_TwoHanded_Handle
  swing1LimitDegrees=80
  swing1Motion=Free
  swing2LimitDegrees=80
  swing2Motion=Free
  twistLimitDegrees=80
  twistMotion=Free
 }
 {
  bone1Name=dyn_eye_r_base
  bone2Name=Melee_TwoHanded_Handle
  swing1LimitDegrees=40
  swing1Motion=Free
  swing2LimitDegrees=40
  swing2Motion=Free
  twistLimitDegrees=40
  twistMotion=Free
 }
 {bone1Name=dyn_ribbon_01, bone2Name=Melee_TwoHanded_Handle, swing1LimitDegrees=90, swing2LimitDegrees=90, twistLimitDegrees=90}
 {bone1Name=dyn_ribbon_02, bone2Name=Melee_TwoHanded_Handle, swing1LimitDegrees=90, swing2LimitDegrees=90, twistLimitDegrees=90}
]
